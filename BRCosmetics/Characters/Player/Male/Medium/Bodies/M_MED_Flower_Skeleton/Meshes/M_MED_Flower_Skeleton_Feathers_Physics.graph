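# Physics asset for the shoulder feathers on the Flower Skeleton male medium body, defining their bodies and joint limits.
# Box bodies on each feather bone hang in one- or two-segment chains from left and right Master bodies, with limited swing and tight twist.
# Both Masters and the upper-arm capsules join spine_05 with free angular motion.
# A separate DYN_ShoulderLFeather body attaches straight to spine_05 with limited motion.
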PhysicsAsset /BRCosmetics/Characters/Player/Male/Medium/Bodies/M_MED_Flower_Skeleton/Meshes/M_MED_Flower_Skeleton_Feathers_Physics
bodies {
 DYN_ShoulderLFeather [{shapeType=Box, extentX=2.9375, extentY=2.25, extentZ=1, center=<1.7596 0 0>}]
 Dyn_Shoulder_FeathersL11_01 [{shapeType=Box, extentX=3.8125, extentY=5.3125, extentZ=1, center=<2.0257 0 0>, rotation=<rot 0 0 -14.0632>}]
 Dyn_Shoulder_FeathersL11_02 [{shapeType=Box, extentX=5.6875, extentY=4.6875, extentZ=1, center=<2.9794 0 -0.0001>, rotation=<rot 0 0 -19.6877>}]
 Dyn_Shoulder_FeathersL1_01 [{shapeType=Box, extentX=3.375, extentY=3.5625, extentZ=1, center=<1.7079 0 0>, rotation=<rot 0 0 36.5631>}]
 Dyn_Shoulder_FeathersL1_02 [{shapeType=Box, extentX=4.8125, extentY=3.1875, extentZ=1, center=<2.5198 0 0>, rotation=<rot 0 0.0001 33.7514>}]
 Dyn_Shoulder_FeathersL2_01 [{shapeType=Box, extentX=4.5625, extentY=4.1875, extentZ=1, center=<2.2029 0 0>, rotation=<rot 0 0 45.001>}]
 Dyn_Shoulder_FeathersL3_01 [{shapeType=Box, extentX=4.75, extentY=1, extentZ=4.8125, center=<2.4414 0 0>, rotation=<rot 0 0 -30.9381>}]
 Dyn_Shoulder_FeathersL3_02 [{shapeType=Box, extentX=5.3125, extentY=1, extentZ=4.5, center=<2.7357 0 0>, rotation=<rot 0 0 -30.9379>}]
 Dyn_Shoulder_FeathersL4_01 [{shapeType=Box, extentX=4.3125, extentY=1, extentZ=3.9375, center=<2.3024 0 0>, rotation=<rot 0 0 -22.4999>}]
 Dyn_Shoulder_FeathersL5_01 [{shapeType=Box, extentX=3.875, extentY=1, extentZ=2.625, center=<1.9313 0 0>}]
 Dyn_Shoulder_FeathersL6_01 [{shapeType=Box, extentX=4.625, extentY=1, extentZ=3.5625, center=<2.4898 0 0>}]
 Dyn_Shoulder_FeathersL7_01 [{shapeType=Box, extentX=6.25, extentY=1, extentZ=3.75, center=<3.2378 0 0>}]
 Dyn_Shoulder_FeathersL8_01 [{shapeType=Box, extentX=5.3125, extentY=1, extentZ=4.9375, center=<2.7979 0 0.0001>, rotation=<rot 0 0 -5.6255>}]
 Dyn_Shoulder_FeathersL8_02 [{shapeType=Box, extentX=5.375, extentY=1, extentZ=4.9375, center=<2.7711 0 -0.0001>, rotation=<rot 0 0 -5.6252>}]
 Dyn_Shoulder_FeathersL9_01 [{shapeType=Box, extentX=5.625, extentY=1, extentZ=3.25, center=<2.8652 -0.0001 0>, rotation=<rot 0 0 30.939>}]
 Dyn_Shoulder_FeathersL_01 [{shapeType=Box, extentX=1, extentY=1, extentZ=1}]
 Dyn_Shoulder_FeathersL_Master [{shapeType=Box, extentX=1, extentY=1, extentZ=1}]
 Dyn_Shoulder_FeathersR10_01 [{shapeType=Box, extentX=4.125, extentY=4.6875, extentZ=1, center=<2.1947 0 0>, rotation=<rot 0 0 5.6252>}]
 Dyn_Shoulder_FeathersR10_02 [{shapeType=Box, extentX=4.875, extentY=4.5, extentZ=1, center=<2.5725 0 0>, rotation=<rot 0 0 5.6252>}]
 Dyn_Shoulder_FeathersR1_01 [{shapeType=Box, extentX=4.125, extentY=3.125, extentZ=1, center=<2.1041 0 -0.0001>, rotation=<rot 0 0 -25.3122>}]
 Dyn_Shoulder_FeathersR2_01 [{shapeType=Box, extentX=4.4375, extentY=1, extentZ=3.625, center=<2.1091 -0.0002 0.0001>, rotation=<rot 0 0 25.3119>}]
 Dyn_Shoulder_FeathersR2_02 [{shapeType=Box, extentX=4.8125, extentY=1, extentZ=4.3125, center=<2.3439 0 0>, rotation=<rot 0 0 33.7527>}]
 Dyn_Shoulder_FeathersR3_01 [{shapeType=Box, extentX=4.4375, extentY=1, extentZ=4.125, center=<2.3406 0 0>}]
 Dyn_Shoulder_FeathersR4_01 [{shapeType=Box, extentX=6.3125, extentY=1, extentZ=4.25, center=<3.1205 0 0>}]
 Dyn_Shoulder_FeathersR5_01 [{shapeType=Box, extentX=6, extentY=1, extentZ=4.125, center=<2.9316 0 0>}]
 Dyn_Shoulder_FeathersR6_01 [{shapeType=Box, extentX=6.375, extentY=1, extentZ=4.25, center=<3.1553 0 0>}]
 Dyn_Shoulder_FeathersR7_01 [{shapeType=Box, extentX=6.375, extentY=1, extentZ=4.8125, center=<3.3018 0 -0.0001>, rotation=<rot 0 0 -59.0623>}]
 Dyn_Shoulder_FeathersR9_01 [{shapeType=Box, extentX=4.375, extentY=3.5, extentZ=1, center=<2.2735 0 0.0001>, rotation=<rot 0 0 11.2507>}]
 Dyn_Shoulder_FeathersR_01 [{shapeType=Box, extentX=1, extentY=1, extentZ=1}]
 Dyn_Shoulder_FeathersR_Master [{shapeType=Box, extentX=1, extentY=1, extentZ=1}]
 spine_05 [{shapeType=Box, extentX=1, extentY=1, extentZ=1}]
 upperarm_l [{shapeType=Capsule, length=18.9156, radius=4.308, center=<13.7733 0.3458 -1.517>, rotation=<rot 0 88.6888 -92.0769>}]
 upperarm_r [{shapeType=Capsule, length=17.495, radius=4.2154, center=<-13.3114 -0.394 1.6198>, rotation=<rot 90 0 0>}]
}
constraints [
 {
  bone1Name=Dyn_Shoulder_FeathersL_01
  bone2Name=Dyn_Shoulder_FeathersL_Master
  swing1LimitDegrees=5
  swing2LimitDegrees=5
  twistLimitDegrees=5
 }
 {
  bone1Name=Dyn_Shoulder_FeathersL11_01
  bone2Name=Dyn_Shoulder_FeathersL_Master
  swing1LimitDegrees=11
  swing2LimitDegrees=25
  twistLimitDegrees=5
 }
 {
  bone1Name=Dyn_Shoulder_FeathersL11_02
  bone2Name=Dyn_Shoulder_FeathersL11_01
  swing1LimitDegrees=8
  swing2LimitDegrees=25
  twistLimitDegrees=5
 }
 {
  bone1Name=Dyn_Shoulder_FeathersL9_01
  bone2Name=Dyn_Shoulder_FeathersL_Master
  swing1LimitDegrees=20
  swing2LimitDegrees=10
  twistLimitDegrees=5
 }
 {
  bone1Name=Dyn_Shoulder_FeathersL8_01
  bone2Name=Dyn_Shoulder_FeathersL_Master
  swing1LimitDegrees=25
  swing2LimitDegrees=10
  twistLimitDegrees=5
 }
 {
  bone1Name=Dyn_Shoulder_FeathersL8_02
  bone2Name=Dyn_Shoulder_FeathersL8_01
  swing1LimitDegrees=25
  swing2LimitDegrees=12
  twistLimitDegrees=5
 }
 {
  bone1Name=Dyn_Shoulder_FeathersL1_01
  bone2Name=Dyn_Shoulder_FeathersL_Master
  swing1LimitDegrees=20
  swing2LimitDegrees=20
  twistLimitDegrees=5
 }
 {
  bone1Name=Dyn_Shoulder_FeathersL1_02
  bone2Name=Dyn_Shoulder_FeathersL1_01
  swing1LimitDegrees=25
  swing2LimitDegrees=20
  twistLimitDegrees=5
 }
 {
  bone1Name=Dyn_Shoulder_FeathersL7_01
  bone2Name=Dyn_Shoulder_FeathersL_Master
  swing1LimitDegrees=25
  swing2LimitDegrees=15
  twistLimitDegrees=5
 }
 {
  bone1Name=Dyn_Shoulder_FeathersL6_01
  bone2Name=Dyn_Shoulder_FeathersL_Master
  swing1LimitDegrees=25
  swing2LimitDegrees=20
  twistLimitDegrees=5
 }
 {
  bone1Name=Dyn_Shoulder_FeathersL5_01
  bone2Name=Dyn_Shoulder_FeathersL_Master
  swing1LimitDegrees=17
  swing2LimitDegrees=17
  twistLimitDegrees=5
 }
 {
  bone1Name=Dyn_Shoulder_FeathersL4_01
  bone2Name=Dyn_Shoulder_FeathersL_Master
  swing1LimitDegrees=25
  swing2LimitDegrees=20
  twistLimitDegrees=5
 }
 {
  bone1Name=Dyn_Shoulder_FeathersL2_01
  bone2Name=Dyn_Shoulder_FeathersL_Master
  swing1LimitDegrees=20
  swing2LimitDegrees=15
  twistLimitDegrees=5
 }
 {
  bone1Name=Dyn_Shoulder_FeathersL3_01
  bone2Name=Dyn_Shoulder_FeathersL_Master
  swing1LimitDegrees=25
  swing2LimitDegrees=20
  twistLimitDegrees=5
 }
 {
  bone1Name=Dyn_Shoulder_FeathersL3_02
  bone2Name=Dyn_Shoulder_FeathersL3_01
  swing1LimitDegrees=20
  swing2LimitDegrees=15
  twistLimitDegrees=5
 }
 {
  bone1Name=Dyn_Shoulder_FeathersR_01
  bone2Name=Dyn_Shoulder_FeathersR_Master
  swing1LimitDegrees=5
  swing2LimitDegrees=5
  twistLimitDegrees=5
 }
 {
  bone1Name=Dyn_Shoulder_FeathersR1_01
  bone2Name=Dyn_Shoulder_FeathersR_Master
  swing1LimitDegrees=20
  swing2LimitDegrees=20
  twistLimitDegrees=5
 }
 {
  bone1Name=Dyn_Shoulder_FeathersR2_01
  bone2Name=Dyn_Shoulder_FeathersR_Master
  swing1LimitDegrees=17
  swing2LimitDegrees=17
  twistLimitDegrees=5
 }
 {
  bone1Name=Dyn_Shoulder_FeathersR2_02
  bone2Name=Dyn_Shoulder_FeathersR2_01
  swing1LimitDegrees=17
  swing2LimitDegrees=17
  twistLimitDegrees=5
 }
 {
  bone1Name=Dyn_Shoulder_FeathersR6_01
  bone2Name=Dyn_Shoulder_FeathersR_Master
  swing1LimitDegrees=20
  swing2LimitDegrees=20
  twistLimitDegrees=5
 }
 {
  bone1Name=Dyn_Shoulder_FeathersR7_01
  bone2Name=Dyn_Shoulder_FeathersR_Master
  swing1LimitDegrees=20
  swing2LimitDegrees=20
  twistLimitDegrees=5
 }
 {
  bone1Name=Dyn_Shoulder_FeathersR5_01
  bone2Name=Dyn_Shoulder_FeathersR_Master
  swing1LimitDegrees=20
  swing2LimitDegrees=20
  twistLimitDegrees=5
 }
 {
  bone1Name=Dyn_Shoulder_FeathersR4_01
  bone2Name=Dyn_Shoulder_FeathersR_Master
  swing1LimitDegrees=20
  swing2LimitDegrees=20
  twistLimitDegrees=5
 }
 {
  bone1Name=Dyn_Shoulder_FeathersR3_01
  bone2Name=Dyn_Shoulder_FeathersR_Master
  swing1LimitDegrees=20
  swing2LimitDegrees=20
  twistLimitDegrees=5
 }
 {
  bone1Name=Dyn_Shoulder_FeathersR9_01
  bone2Name=Dyn_Shoulder_FeathersR_Master
  swing1LimitDegrees=20
  swing2LimitDegrees=20
  twistLimitDegrees=5
 }
 {
  bone1Name=Dyn_Shoulder_FeathersR10_01
  bone2Name=Dyn_Shoulder_FeathersR_Master
  swing1LimitDegrees=20
  swing2LimitDegrees=20
  twistLimitDegrees=5
 }
 {
  bone1Name=Dyn_Shoulder_FeathersR10_02
  bone2Name=Dyn_Shoulder_FeathersR10_01
  swing1LimitDegrees=20
  swing2LimitDegrees=20
  twistLimitDegrees=5
 }
 {
  bone1Name=upperarm_l
  bone2Name=spine_05
  swing1LimitDegrees=45
  swing1Motion=Free
  swing2LimitDegrees=45
  swing2Motion=Free
  twistLimitDegrees=45
  twistMotion=Free
 }
 {
  bone1Name=Dyn_Shoulder_FeathersL_Master
  bone2Name=spine_05
  swing1LimitDegrees=45
  swing1Motion=Free
  swing2LimitDegrees=45
  swing2Motion=Free
  twistLimitDegrees=45
  twistMotion=Free
 }
 {
  bone1Name=upperarm_r
  bone2Name=spine_05
  swing1LimitDegrees=45
  swing1Motion=Free
  swing2LimitDegrees=45
  swing2Motion=Free
  twistLimitDegrees=45
  twistMotion=Free
 }
 {
  bone1Name=Dyn_Shoulder_FeathersR_Master
  bone2Name=spine_05
  swing1LimitDegrees=45
  swing1Motion=Free
  swing2LimitDegrees=45
  swing2Motion=Free
  twistLimitDegrees=45
  twistMotion=Free
 }
 {bone1Name=DYN_ShoulderLFeather, bone2Name=spine_05, swing1LimitDegrees=20, swing2LimitDegrees=15, twistLimitDegrees=5}
]
