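# Physics asset for the flower accessories on the M_MED_Flower_Skeleton player body cosmetic, covering the shoulder flowers and the arm and foot roses.
# Flowers and leaves use boxes and thorns use capsules, with swing and twist limits that are generally tighter for thorns.
# They constrain to small anchor boxes on spine_05, forearms and feet.
# Some right-arm pieces route through KinematicHold bodies, which have free angular motion relative to the twist bone.
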PhysicsAsset /BRCosmetics/Characters/Player/Male/Medium/Bodies/M_MED_Flower_Skeleton/Meshes/M_MED_Flower_Skeleton_Flower_Physics
bodies {
 DYN_FlowerShoulder_01 [{shapeType=Box, extentX=2.5, extentY=2.625, extentZ=3.0625, center=<1.4438 0 0>}]
 DYN_FlowerShoulder_01_Leaf [{shapeType=Box, extentX=3, extentY=1, extentZ=3.5, center=<1.5 0 0>}]
 DYN_FlowerShoulder_02 [{shapeType=Box, extentX=3.4375, extentY=3, extentZ=2.6875, center=<1.9496 0 0>}]
 DYN_FlowerShoulder_03 [{shapeType=Box, extentX=3.0625, extentY=3.25, extentZ=3.3125, center=<1.815 0 0>}]
 DYN_FlowerShoulder_04 [{shapeType=Box, extentX=3.4375, extentY=3.3125, extentZ=3.3125, center=<2.0301 0 0>}]
 DYN_FlowerShoulder_05 [{shapeType=Box, extentX=3, extentY=3.5, extentZ=3.3125, center=<1.7695 0 0>}]
 DYN_FlowerShoulder_06 [{shapeType=Box, extentX=2.875, extentY=2.9375, extentZ=2.875, center=<1.5319 -0.0003 0>}]
 DYN_FlowerShoulder_07 [{shapeType=Box, extentX=2.25, extentY=3.0625, extentZ=3.0625, center=<1.314 0 0>}]
 DYN_FlowerShoulder_08 [{shapeType=Box, extentX=3.4375, extentY=2.9375, extentZ=2.8125, center=<1.6002 0 0>}]
 DYN_FlowerShoulder_09 [{shapeType=Box, extentX=3.9375, extentY=3, extentZ=3.1875, center=<2.0287 0 0>}]
 DYN_FlowerShoulder_10 [{shapeType=Box, extentX=3.3125, extentY=3.5, extentZ=3.4375, center=<1.7659 0 0>}]
 DYN_FlowerShoulder_11 [{shapeType=Box, extentX=3.4375, extentY=3.3125, extentZ=3.9375, center=<2.0656 -0.0001 0>}]
 DYN_FlowerShoulder_12 [{shapeType=Box, extentX=2.6875, extentY=3.4375, extentZ=3, center=<1.2876 0 0>}]
 DYN_FlowerShoulder_13 [{shapeType=Box, extentX=2.6875, extentY=3.4375, extentZ=3.0625, center=<1.5554 0 0>}]
 DYN_FlowerShoulder_13_Leaf [{shapeType=Box, extentX=2.9375, extentY=1, extentZ=3.375, center=<1.3854 0.0002 0>}]
 Dyn_ArmRoseL1 [{shapeType=Box, extentX=3.5625, extentY=2.375, extentZ=2.5625, center=<1.9643 0 0>}]
 Dyn_ArmRoseL2 [{shapeType=Box, extentX=3.4375, extentY=2.375, extentZ=2.3125, center=<1.6937 0 0>}]
 Dyn_ArmRoseL_Thorn1 [{shapeType=Capsule, length=9.4093, radius=1.2744, center=<6.0842 -0.0003 0>, rotation=<rot 90 0 0>}]
 Dyn_ArmRoseL_Thorn2 [{shapeType=Capsule, length=3.882, radius=0.6728, center=<2.5886 0 0>, rotation=<rot 90 0 0>}]
 Dyn_ArmRoseR1 [{shapeType=Box, extentX=2.5, extentY=2.5, extentZ=2.375, center=<1.558 -0.0001 0>}]
 Dyn_ArmRoseR1_KinematicHold [{shapeType=Box, extentX=1, extentY=1, extentZ=1}]
 Dyn_ArmRoseR1_Leaf1 [{shapeType=Box, extentX=2.0625, extentY=2.5625, extentZ=1, center=<1.0631 0 0>}]
 Dyn_ArmRoseR1_Leaf2 [{shapeType=Box, extentX=2.3125, extentY=1.875, extentZ=1, center=<1.2857 0 0.0002>}]
 Dyn_ArmRoseR2 [{shapeType=Box, extentX=3.1875, extentY=3.125, extentZ=1.9375, center=<1.8463 0 0>}]
 Dyn_ArmRoseR2KinematicHold [{shapeType=Box, extentX=1, extentY=1, extentZ=1}]
 Dyn_ArmRoseR2_Leaf1 [{shapeType=Box, extentX=1.625, extentY=0.5, extentZ=1.625, center=<0.9656 0 0>}]
 Dyn_ArmRoseR2_Leaf2 [{shapeType=Box, extentX=1.9375, extentY=0.5, extentZ=2.25, center=<1.0259 0.0001 -0.0002>, rotation=<rot -0.0001 0 50.6269>}]
 Dyn_ArmRoseR2_Leaf3 [{shapeType=Box, extentX=1, extentY=0.6875, extentZ=1.6875, center=<0.5635 -0.0001 0.0001>, rotation=<rot 0 0 39.3746>}]
 Dyn_ArmRoseR_Thorn1 [{shapeType=Capsule, length=3.63, radius=0.7237, center=<2.5862 0 0.0001>, rotation=<rot 90 0 0>}]
 Dyn_FootRoseL [{shapeType=Box, extentX=3, extentY=2.5625, extentZ=2.875, center=<1.6292 0 0>}]
 Dyn_FootRoseL_Thorn1 [{shapeType=Capsule, length=8.13, radius=0.8487, center=<4.7676 0 0>, rotation=<rot 90 0 0>}]
 Dyn_FootRoseL_Thorn2 [{shapeType=Capsule, length=6.0675, radius=0.5987, center=<3.659 -0.0001 0>, rotation=<rot 90 0 0>}]
 Dyn_FootRoseR [{shapeType=Box, extentX=2.9375, extentY=2.5, extentZ=1.9375, center=<1.7242 0 0>}]
 Dyn_FootRoseR_Thorn1 [{shapeType=Capsule, length=9.13, radius=0.755, center=<5.1699 0.0001 0>, rotation=<rot 90 0 0>}]
 Dyn_FootRoseR_Thorn2 [{shapeType=Capsule, length=3.0363, radius=0.6612, center=<2.1415 0 0>, rotation=<rot 90 0 0>}]
 foot_l [{shapeType=Box, extentX=1, extentY=1, extentZ=1}]
 foot_r [{shapeType=Box, extentX=1, extentY=1, extentZ=1}]
 lowerarm_l [{shapeType=Box, extentX=1, extentY=1, extentZ=1}]
 lowerarm_r [{shapeType=Box, extentX=1, extentY=1, extentZ=1}]
 lowerarm_twist_02_r [{shapeType=Box, extentX=1, extentY=1, extentZ=1}]
 spine_05 [{shapeType=Box, extentX=1, extentY=1, extentZ=1}]
}
constraints [
 {bone1Name=Dyn_ArmRoseR2, bone2Name=Dyn_ArmRoseR2KinematicHold, swing1LimitDegrees=12, swing2LimitDegrees=12, twistLimitDegrees=5}
 {bone1Name=Dyn_ArmRoseR_Thorn1, bone2Name=Dyn_ArmRoseR2KinematicHold, swing1LimitDegrees=15, swing2LimitDegrees=15, twistLimitDegrees=5}
 {bone1Name=Dyn_ArmRoseR1, bone2Name=Dyn_ArmRoseR1_KinematicHold, swing1LimitDegrees=12, swing2LimitDegrees=12, twistLimitDegrees=5}
 {bone1Name=Dyn_ArmRoseR1_Leaf1, bone2Name=Dyn_ArmRoseR1_KinematicHold, swing1LimitDegrees=20, swing2LimitDegrees=45, twistLimitDegrees=5}
 {bone1Name=Dyn_ArmRoseL_Thorn2, bone2Name=lowerarm_l, swing1LimitDegrees=10, swing2LimitDegrees=10, twistLimitDegrees=5}
 {bone1Name=Dyn_ArmRoseL1, bone2Name=lowerarm_l, swing1LimitDegrees=32, swing2LimitDegrees=32, twistLimitDegrees=5}
 {bone1Name=Dyn_ArmRoseL2, bone2Name=lowerarm_l, swing1LimitDegrees=30, swing2LimitDegrees=30, twistLimitDegrees=5}
 {bone1Name=Dyn_ArmRoseL_Thorn1, bone2Name=lowerarm_l, swing1LimitDegrees=8, swing2LimitDegrees=8, twistLimitDegrees=5}
 {bone1Name=Dyn_ArmRoseR2_Leaf3, bone2Name=lowerarm_twist_02_r, swing1LimitDegrees=30, swing2LimitDegrees=30, twistLimitDegrees=5}
 {bone1Name=Dyn_ArmRoseR2_Leaf2, bone2Name=lowerarm_twist_02_r, swing1LimitDegrees=30, swing2LimitDegrees=30, twistLimitDegrees=5}
 {bone1Name=Dyn_ArmRoseR2_Leaf1, bone2Name=lowerarm_twist_02_r, swing1LimitDegrees=15, swing2LimitDegrees=15, twistLimitDegrees=5}
 {
  bone1Name=Dyn_ArmRoseR2KinematicHold
  bone2Name=lowerarm_twist_02_r
  swing1LimitDegrees=45
  swing1Motion=Free
  swing2LimitDegrees=45
  swing2Motion=Free
  twistLimitDegrees=45
  twistMotion=Free
 }
 {
  bone1Name=Dyn_ArmRoseR1_KinematicHold
  bone2Name=lowerarm_twist_02_r
  swing1LimitDegrees=45
  swing1Motion=Free
  swing2LimitDegrees=45
  swing2Motion=Free
  twistLimitDegrees=45
  twistMotion=Free
 }
 {
  bone1Name=lowerarm_twist_02_r
  bone2Name=lowerarm_r
  swing1LimitDegrees=45
  swing1Motion=Free
  swing2LimitDegrees=45
  swing2Motion=Free
  twistLimitDegrees=45
  twistMotion=Free
 }
 {bone1Name=Dyn_ArmRoseR1_Leaf2, bone2Name=lowerarm_r, swing1LimitDegrees=20, swing2LimitDegrees=40, twistLimitDegrees=5}
 {bone1Name=Dyn_FootRoseL, bone2Name=foot_l, swing1LimitDegrees=30, swing2LimitDegrees=30, twistLimitDegrees=5}
 {bone1Name=Dyn_FootRoseL_Thorn1, bone2Name=foot_l, swing1LimitDegrees=12, swing2LimitDegrees=12, twistLimitDegrees=5}
 {bone1Name=Dyn_FootRoseL_Thorn2, bone2Name=foot_l, swing1LimitDegrees=8, swing2LimitDegrees=8, twistLimitDegrees=5}
 {bone1Name=Dyn_FootRoseR, bone2Name=foot_r, swing1LimitDegrees=30, swing2LimitDegrees=30, twistLimitDegrees=5}
 {bone1Name=Dyn_FootRoseR_Thorn1, bone2Name=foot_r, swing1LimitDegrees=12, swing2LimitDegrees=12, twistLimitDegrees=5}
 {bone1Name=Dyn_FootRoseR_Thorn2, bone2Name=foot_r, swing1LimitDegrees=8, swing2LimitDegrees=8, twistLimitDegrees=5}
 {
  bone1Name=lowerarm_l
  bone2Name=spine_05
  swing1LimitDegrees=45
  swing1Motion=Free
  swing2LimitDegrees=45
  swing2Motion=Free
  twistLimitDegrees=45
  twistMotion=Free
 }
 {
  bone1Name=lowerarm_r
  bone2Name=spine_05
  swing1LimitDegrees=45
  swing1Motion=Free
  swing2LimitDegrees=45
  swing2Motion=Free
  twistLimitDegrees=45
  twistMotion=Free
 }
 {bone1Name=DYN_FlowerShoulder_13_Leaf, bone2Name=spine_05, swing1LimitDegrees=20, swing2LimitDegrees=20, twistLimitDegrees=20}
 {bone1Name=DYN_FlowerShoulder_10, bone2Name=spine_05, swing1LimitDegrees=30, swing2LimitDegrees=30, twistLimitDegrees=5}
 {bone1Name=DYN_FlowerShoulder_09, bone2Name=spine_05, swing1LimitDegrees=30, swing2LimitDegrees=30, twistLimitDegrees=5}
 {bone1Name=DYN_FlowerShoulder_08, bone2Name=spine_05, swing1LimitDegrees=30, swing2LimitDegrees=30, twistLimitDegrees=5}
 {bone1Name=DYN_FlowerShoulder_07, bone2Name=spine_05, swing1LimitDegrees=30, swing2LimitDegrees=30, twistLimitDegrees=5}
 {bone1Name=DYN_FlowerShoulder_06, bone2Name=spine_05, swing1LimitDegrees=30, swing2LimitDegrees=30, twistLimitDegrees=5}
 {bone1Name=DYN_FlowerShoulder_05, bone2Name=spine_05, swing1LimitDegrees=30, swing2LimitDegrees=30, twistLimitDegrees=5}
 {bone1Name=DYN_FlowerShoulder_04, bone2Name=spine_05, swing1LimitDegrees=30, swing2LimitDegrees=30, twistLimitDegrees=5}
 {bone1Name=DYN_FlowerShoulder_11, bone2Name=spine_05, swing1LimitDegrees=30, swing2LimitDegrees=30, twistLimitDegrees=5}
 {bone1Name=DYN_FlowerShoulder_12, bone2Name=spine_05, swing1LimitDegrees=30, swing2LimitDegrees=30, twistLimitDegrees=5}
 {bone1Name=DYN_FlowerShoulder_13, bone2Name=spine_05, swing1LimitDegrees=30, swing2LimitDegrees=30, twistLimitDegrees=5}
 {bone1Name=DYN_FlowerShoulder_01, bone2Name=spine_05, swing1LimitDegrees=30, swing2LimitDegrees=30, twistLimitDegrees=5}
 {bone1Name=DYN_FlowerShoulder_02, bone2Name=spine_05, swing1LimitDegrees=30, swing2LimitDegrees=30, twistLimitDegrees=5}
 {bone1Name=DYN_FlowerShoulder_03, bone2Name=spine_05, swing1LimitDegrees=30, swing2LimitDegrees=30, twistLimitDegrees=5}
 {bone1Name=DYN_FlowerShoulder_01_Leaf, bone2Name=spine_05, swing1LimitDegrees=20, swing2LimitDegrees=20, twistLimitDegrees=20}
]
