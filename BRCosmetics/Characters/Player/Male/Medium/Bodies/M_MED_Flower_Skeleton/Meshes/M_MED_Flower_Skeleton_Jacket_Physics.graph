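# Physics asset for the jacket of the male medium Flower Skeleton player cosmetic, defining secondary motion for its jacket, zipper and belt bones.
# Two six-link box chains (Jacket01, Jacket06) hang from spine_05 with mostly tight angular limits, and a looser zipper ends Jacket01.
# Four looser two-link jacket chains hang from spine_03, and a two-link belt hangs from pelvis.
# Pelvis, spine and thigh bodies, mostly capsules, supply body shapes joined by angularly free constraints.
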PhysicsAsset /BRCosmetics/Characters/Player/Male/Medium/Bodies/M_MED_Flower_Skeleton/Meshes/M_MED_Flower_Skeleton_Jacket_Physics
bodies {
 DYN_Jacket01_1 [{shapeType=Box, extentX=3.5625, extentY=4, extentZ=1, center=<1.7164 0 0>, rotation=<rot 0 0 -45.0014>}]
 DYN_Jacket01_2 [{shapeType=Box, extentX=3.5625, extentY=1.9375, extentZ=3.875, center=<2.0627 0 0>, rotation=<rot 0 0 28.1263>}]
 DYN_Jacket01_3 [{shapeType=Box, extentX=3.5, extentY=1.625, extentZ=4.3125, center=<1.8614 0.0001 0>, rotation=<rot 0 0 28.1254>}]
 DYN_Jacket01_4 [{shapeType=Box, extentX=3.25, extentY=1.75, extentZ=4.25, center=<1.8244 0 0>, rotation=<rot 0 0 28.1258>}]
 DYN_Jacket01_5 [{shapeType=Box, extentX=3.0625, extentY=1.8125, extentZ=8.5625, center=<1.6796 -0.0001 0>, rotation=<rot 0 0 8.4382>}]
 DYN_Jacket01_6 [{shapeType=Box, extentX=6.8125, extentY=3.4375, extentZ=16.1875, center=<3.4172 0 0>, rotation=<rot 0 0 5.6253>}]
 DYN_Jacket02_1 [{shapeType=Box, extentX=2, extentY=1.5, extentZ=5.8125, center=<1.4038 0 0>, rotation=<rot 0 0 2.8127>}]
 DYN_Jacket02_2 [{shapeType=Box, extentX=4, extentY=2.1875, extentZ=5.4375, center=<2.0984 0 0>}]
 DYN_Jacket03_1 [{shapeType=Box, extentX=3.6875, extentY=1.75, extentZ=5.875, center=<1.8867 0 0>, rotation=<rot 0 0 -11.2508>}]
 DYN_Jacket03_2 [{shapeType=Box, extentX=3.5625, extentY=2.125, extentZ=6.125, center=<1.9751 0 0>, rotation=<rot 0 0 5.6252>}]
 DYN_Jacket04_1 [{shapeType=Box, extentX=3.0625, extentY=0.9375, extentZ=6.3125, center=<1.6663 0 0>, rotation=<rot 0 0 -11.251>}]
 DYN_Jacket04_2 [{shapeType=Box, extentX=3.1875, extentY=1.125, extentZ=6.0625, center=<1.8386 0 0.0001>, rotation=<rot 0 0 -39.3756>}]
 DYN_Jacket05_1 [{shapeType=Box, extentX=2.625, extentY=1.1875, extentZ=5.625, center=<1.3879 0 0>}]
 DYN_Jacket05_2 [{shapeType=Box, extentX=3.75, extentY=2.125, extentZ=6.0625, center=<1.9773 0 0>}]
 DYN_Jacket06_1 [{shapeType=Box, extentX=3.1875, extentY=5.125, extentZ=1, center=<1.8255 0 0>, rotation=<rot 0 0 36.5618>}]
 DYN_Jacket06_2 [{shapeType=Box, extentX=3.75, extentY=5.3125, extentZ=1.8125, center=<1.937 0 0>, rotation=<rot 0 0 30.9367>}]
 DYN_Jacket06_3 [{shapeType=Box, extentX=3.5625, extentY=5.1875, extentZ=1.3125, center=<1.8835 0 0>, rotation=<rot 0 0 30.9385>}]
 DYN_Jacket06_4 [{shapeType=Box, extentX=3.25, extentY=5.3125, extentZ=1.4375, center=<1.7917 0 0>, rotation=<rot 0 0 33.7493>}]
 DYN_Jacket06_5 [{shapeType=Box, extentX=2.9375, extentY=5.5, extentZ=1.4375, center=<1.6239 0 0>, rotation=<rot 0 0 36.5636>}]
 DYN_Jacket06_6 [{shapeType=Box, extentX=5.5625, extentY=5.4375, extentZ=2.125, center=<2.7746 -0.0001 0>, rotation=<rot 0 0 36.5637>}]
 Dyn_Zipper [{shapeType=Box, extentX=2.625, extentY=1, extentZ=1.875, center=<1.4472 0 0>}]
 Dyn_belt_01 [{shapeType=Box, extentX=2.6875, extentY=1, extentZ=4.625, center=<1.3739 0 0>}]
 Dyn_belt_02 [{shapeType=Box, extentX=2.8125, extentY=1, extentZ=4.25, center=<1.7321 0 0>}]
 pelvis [{shapeType=Capsule, length=13.7282, radius=9.4934, center=<-1.693 0 -0.0001>, rotation=<rot 0 8.4387 0>}]
 spine_01 [
  {shapeType=Capsule, length=13.1462, radius=9.3767, center=<0 0.9758 0.0001>}
  {shapeType=Box, extentX=38.1875, extentY=3.4375, extentZ=27.25, center=<5.4373 -5.7297 15.4137>, rotation=<rot -0.0001 0.0003 -19.6882>}
 ]
 spine_02 [
  {shapeType=Capsule, length=12.1257, radius=8.659, center=<4.3219 1.6965 0.2899>, rotation=<rot 0 -179.033 173.6116>}
  {shapeType=Capsule, length=4.9062, radius=2.6875, center=<0.6927 -0.5633 11.6504>, rotation=<rot 0 0 -95.6251>}
  {shapeType=Capsule, length=5.625, radius=9.4375, center=<-0.0001 3.4134 -0.1101>}
 ]
 spine_03 [{shapeType=Capsule, length=7.4237, radius=12.5909, center=<0 2.3711 0>, rotation=<rot 90 176.4237 167.986>}]
 spine_05 [{shapeType=Box, extentX=1, extentY=1, extentZ=1}]
 thigh_l [{shapeType=Capsule, length=21.7704, radius=7.1144, center=<-16.6178 1.1803 1.1771>, rotation=<rot 90 0 0>}]
 thigh_r [{shapeType=Capsule, length=15.2704, radius=7.1144, center=<13.3378 -1.1801 -1.1771>, rotation=<rot 90 0 0>}]
}
constraints [
 {bone1Name=DYN_Jacket01_2, bone2Name=DYN_Jacket01_1, swing1LimitDegrees=5, swing2LimitDegrees=5, twistLimitDegrees=5}
 {bone1Name=DYN_Jacket01_3, bone2Name=DYN_Jacket01_2, swing1LimitDegrees=10, swing2LimitDegrees=8, twistLimitDegrees=7}
 {bone1Name=DYN_Jacket01_4, bone2Name=DYN_Jacket01_3, swing1LimitDegrees=10, swing2LimitDegrees=8, twistLimitDegrees=7}
 {bone1Name=DYN_Jacket01_5, bone2Name=DYN_Jacket01_4, swing1LimitDegrees=10, swing2LimitDegrees=8, twistLimitDegrees=7}
 {bone1Name=DYN_Jacket01_6, bone2Name=DYN_Jacket01_5, swing1LimitDegrees=12, swing2LimitDegrees=8, twistLimitDegrees=7}
 {bone1Name=Dyn_Zipper, bone2Name=DYN_Jacket01_6, swing1LimitDegrees=45, swing2LimitDegrees=45, twistLimitDegrees=5}
 {bone1Name=DYN_Jacket06_2, bone2Name=DYN_Jacket06_1, swing1LimitDegrees=5, swing2LimitDegrees=5, twistLimitDegrees=5}
 {bone1Name=DYN_Jacket06_3, bone2Name=DYN_Jacket06_2, swing1LimitDegrees=10, swing2LimitDegrees=8, twistLimitDegrees=7}
 {bone1Name=DYN_Jacket06_4, bone2Name=DYN_Jacket06_3, swing1LimitDegrees=10, swing2LimitDegrees=8, twistLimitDegrees=7}
 {bone1Name=DYN_Jacket06_5, bone2Name=DYN_Jacket06_4, swing1LimitDegrees=10, swing2LimitDegrees=8, twistLimitDegrees=7}
 {bone1Name=DYN_Jacket06_6, bone2Name=DYN_Jacket06_5, swing1LimitDegrees=10, swing2LimitDegrees=8, twistLimitDegrees=7}
 {bone1Name=DYN_Jacket02_2, bone2Name=DYN_Jacket02_1, swing1LimitDegrees=25, swing2LimitDegrees=20, twistLimitDegrees=7}
 {bone1Name=DYN_Jacket03_2, bone2Name=DYN_Jacket03_1, swing1LimitDegrees=40, swing2LimitDegrees=20, twistLimitDegrees=7}
 {bone1Name=DYN_Jacket04_2, bone2Name=DYN_Jacket04_1, swing1LimitDegrees=30, swing2LimitDegrees=20, twistLimitDegrees=7}
 {bone1Name=DYN_Jacket05_2, bone2Name=DYN_Jacket05_1, swing1LimitDegrees=25, swing2LimitDegrees=20, twistLimitDegrees=7}
 {bone1Name=Dyn_belt_02, bone2Name=Dyn_belt_01, swing1LimitDegrees=20, swing2LimitDegrees=10, twistLimitDegrees=5}
 {bone1Name=DYN_Jacket01_1, bone2Name=spine_05, swing1LimitDegrees=5, swing2LimitDegrees=5, twistLimitDegrees=5}
 {bone1Name=DYN_Jacket06_1, bone2Name=spine_05, swing1LimitDegrees=5, swing2LimitDegrees=5, twistLimitDegrees=5}
 {bone1Name=DYN_Jacket02_1, bone2Name=spine_03, swing1LimitDegrees=35, swing2LimitDegrees=20, twistLimitDegrees=7}
 {bone1Name=DYN_Jacket03_1, bone2Name=spine_03, swing1LimitDegrees=40, swing2LimitDegrees=20, twistLimitDegrees=7}
 {bone1Name=DYN_Jacket04_1, bone2Name=spine_03, swing1LimitDegrees=40, swing2LimitDegrees=20, twistLimitDegrees=7}
 {bone1Name=DYN_Jacket05_1, bone2Name=spine_03, swing1LimitDegrees=45, swing2LimitDegrees=20, twistLimitDegrees=7}
 {
  bone1Name=spine_05
  bone2Name=spine_03
  swing1LimitDegrees=45
  swing1Motion=Free
  swing2LimitDegrees=45
  swing2Motion=Free
  twistLimitDegrees=45
  twistMotion=Free
 }
 {
  bone1Name=spine_03
  bone2Name=spine_02
  swing1LimitDegrees=45
  swing1Motion=Free
  swing2LimitDegrees=45
  swing2Motion=Free
  twistLimitDegrees=45
  twistMotion=Free
 }
 {bone1Name=Dyn_belt_01, bone2Name=pelvis, swing1LimitDegrees=20, swing2LimitDegrees=10, twistLimitDegrees=5}
 {
  bone1Name=thigh_r
  bone2Name=pelvis
  swing1LimitDegrees=45
  swing1Motion=Free
  swing2LimitDegrees=45
  swing2Motion=Free
  twistLimitDegrees=45
  twistMotion=Free
 }
 {
  bone1Name=spine_02
  bone2Name=spine_01
  swing1LimitDegrees=45
  swing1Motion=Free
  swing2LimitDegrees=45
  swing2Motion=Free
  twistLimitDegrees=45
  twistMotion=Free
 }
 {
  bone1Name=thigh_l
  bone2Name=pelvis
  swing1LimitDegrees=45
  swing1Motion=Free
  swing2LimitDegrees=45
  swing2Motion=Free
  twistLimitDegrees=45
  twistMotion=Free
 }
 {
  bone1Name=spine_01
  bone2Name=pelvis
  swing1LimitDegrees=45
  swing1Motion=Free
  swing2LimitDegrees=45
  swing2Motion=Free
  twistLimitDegrees=45
  twistMotion=Free
 }
]
