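# Animation Blueprint for the Flower Skeleton male cosmetic body, adding secondary motion and pose correctives to flower, jacket, feather and thorn bones.
# Each of two branches copies the parent mesh pose and chains RigidBody simulations; a bool blend selects the skydiving/parachute branch or the regular one.
# Additive bone offsets and jacket constraints respond to targeting, jumping, crouching, arm and head poses.
# The event graph derives arm and head alphas from bone rotation deltas relative to spine_05.
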AnimBlueprint /BRCosmetics/Characters/Player/Male/Medium/Bodies/M_MED_Flower_Skeleton/Meshes/M_MED_Flower_Skeleton_AnimBlueprint
target_skeleton=&/M_MED_Flower_Skeleton_Skeleton
graphs [
 AnimGraph {nodes [
   "Transform (Modify) Bone - Bone: Dyn_Shoulder_FeathersR_Master"(Alpha=Get Right_Arm_ADS_Down() -> (Right_Arm_ADS_Down:double), Translation="-3,0.000,0", Rotation="0,0,0", Scale="1,1,1") {
    rotationMode=BMM_Additive
    rotationSpace=BCS_BoneSpace
    translationMode=BMM_Additive
    translationSpace=BCS_BoneSpace
    alphaBoolBlend {blendOption=Linear}
    boneToModify {boneName=Dyn_Shoulder_FeathersR_Master}
    translation {x=-3}
   }
   n1 = RigidBody(ComponentPose=Local To Component(LocalPose=Copy Pose From Mesh() {bCopyCurves=true, bUseAttachedParent=true})) {
    bOverrideWorldGravity=true
    overridePhysicsAsset=&/M_MED_Flower_Skeleton_Flower_Physics
    simulationSpace=BaseBoneSpace
    alphaBoolBlend {blendOption=Linear}
    baseBoneRef {boneName=pelvis}
    componentLinearAccScale=<0.3 0.3 0.3>
    componentLinearVelScale=<0.6 0.6 0.6>
    overrideWorldGravity {z=-980}
   }
   n2 = RigidBody(ComponentPose=@n1) {
    bOverrideWorldGravity=true
    overridePhysicsAsset=&/M_MED_Flower_Skeleton_Jacket_Physics
    simulationSpace=BaseBoneSpace
    alphaBoolBlend {blendOption=Linear}
    baseBoneRef {boneName=pelvis}
    componentLinearAccScale=<0.3 0.3 0.3>
    componentLinearVelScale=<0.6 0.6 0.6>
    overrideWorldGravity {z=-980}
   }
   n3 = RigidBody(ComponentPose=@n2, Alpha=0.7) {
    bOverrideWorldGravity=true
    overridePhysicsAsset=&/M_MED_Flower_Skeleton_Feathers_Physics
    simulationSpace=BaseBoneSpace
    alphaBoolBlend {blendOption=Linear}
    baseBoneRef {boneName=pelvis}
    componentLinearAccScale=<0.3 0.3 0.3>
    componentLinearVelScale=<0.6 0.6 0.6>
    overrideWorldGravity {z=-480}
   }
   n4 = "To Float (Boolean)"(InBool=Get bIsSkyDiving())
   n5 = "Transform (Modify) Bone - Bone: Dyn_Shoulder_FeathersL_Master"(ComponentPose=@n3, Alpha=@n4, Translation="0,0,0", Rotation="0,50,0", Scale="1,1,1") {
    rotationMode=BMM_Additive
    translationMode=BMM_Additive
    alphaBoolBlend {blendOption=Linear}
    boneToModify {boneName=Dyn_Shoulder_FeathersL_Master}
    rotation {yaw=50}
   }
   n6 = "Transform (Modify) Bone - Bone: Dyn_Shoulder_FeathersR_Master"(ComponentPose=@n5, Alpha=@n4, Translation="0,0,0", Rotation="0,-50,0", Scale="1,1,1") {
    rotationMode=BMM_Additive
    translationMode=BMM_Additive
    alphaBoolBlend {blendOption=Linear}
    boneToModify {boneName=Dyn_Shoulder_FeathersR_Master}
    rotation {yaw=-50}
   }
   n7 = "To Float (Boolean)"(InBool=Get bIsParachuteOpened())
   n8 = "Transform (Modify) Bone - Bone: Dyn_Shoulder_FeathersL_Master"(ComponentPose=@n6, Alpha=@n7, Translation="0,-6,0", Rotation="0,-30,-70", Scale="1,1,1") {
    rotationMode=BMM_Additive
    translationMode=BMM_Additive
    alphaBoolBlend {blendOption=Linear}
    boneToModify {boneName=Dyn_Shoulder_FeathersL_Master}
    rotation {roll=-70, yaw=-30}
    translation {y=-6}
   }
   n9 = "Transform (Modify) Bone - Bone: Dyn_Shoulder_FeathersR_Master"(ComponentPose=@n8, Alpha=@n7, Translation="0,-6,0", Rotation="0,50,-70", Scale="1,1,1") {
    rotationMode=BMM_Additive
    translationMode=BMM_Additive
    alphaBoolBlend {blendOption=Linear}
    boneToModify {boneName=Dyn_Shoulder_FeathersR_Master}
    rotation {roll=-70, yaw=50}
    translation {y=-6}
   }
   n10 = Component To Local(ComponentPose=@n9)
   n11 = RigidBody(ComponentPose=Local To Component(LocalPose=Copy Pose From Mesh() {bCopyCurves=true, bUseAttachedParent=true})) {
    bOverrideWorldGravity=true
    overridePhysicsAsset=&M_MED_Flower_Skeleton_Flower_Physics
    simulationSpace=BaseBoneSpace
    alphaBoolBlend {blendOption=Linear}
    baseBoneRef {boneName=pelvis}
    componentLinearAccScale=<0.3 0.3 0.3>
    componentLinearVelScale=<0.6 0.6 0.6>
    overrideWorldGravity {z=-980}
   }
   n12 = RigidBody(ComponentPose=@n11) {
    bOverrideWorldGravity=true
    overridePhysicsAsset=&M_MED_Flower_Skeleton_Jacket_Physics
    simulationSpace=BaseBoneSpace
    alphaBoolBlend {blendOption=Linear}
    baseBoneRef {boneName=pelvis}
    componentLinearAccScale=<0.3 0.3 0.3>
    componentLinearVelScale=<0.6 0.6 0.6>
    overrideWorldGravity {z=-980}
   }
   n13 = RigidBody(ComponentPose=@n12) {
    bOverrideWorldGravity=true
    overridePhysicsAsset=&M_MED_Flower_Skeleton_Feathers_Physics
    simulationSpace=BaseBoneSpace
    alphaBoolBlend {blendOption=Linear}
    baseBoneRef {boneName=pelvis}
    componentLinearAccScale=<0.3 0.3 0.3>
    componentLinearVelScale=<0.6 0.6 0.6>
    overrideWorldGravity {z=-480}
   }
   n14 = Get Left_Arm_ADS() -> (Left_Arm_ADS:double)
   n15 = "Transform (Modify) Bone - Bone: Dyn_Shoulder_FeathersL_Master"(ComponentPose=@n13, Alpha=@n14, Translation="0.000,-1,0.000", Rotation="0.000,0.000,-20", Scale="1,1,1") {
    rotationMode=BMM_Additive
    translationMode=BMM_Additive
    alphaBoolBlend {blendOption=Linear}
    boneToModify {boneName=Dyn_Shoulder_FeathersL_Master}
    rotation {roll=-20}
    translation {y=-1}
   }
   n16 = "Transform (Modify) Bone - Bone: Dyn_Shoulder_FeathersL_Master1"(ComponentPose=@n15, Alpha=@n14, Translation="0.000,-3,5.000", Rotation="-14,-25,0", Scale="1,1,1") {
    rotationMode=BMM_Additive
    translationMode=BMM_Additive
    alphaBoolBlend {blendOption=Linear}
    boneToModify {boneName=Dyn_Shoulder_FeathersL_Master1}
    rotation {pitch=-14, yaw=-25}
    translation {y=-3, z=5}
   }
   n17 = "To Float (Boolean)"(InBool=Get bIsTargeting())
   n18 = "Transform (Modify) Bone - Bone: DYN_ThornShoulder_03"(ComponentPose=@n16, Alpha=@n17, Translation="0,0,0", Rotation="180,0,0", Scale="1,1,1") {rotationMode=BMM_Additive, alphaBoolBlend {blendOption=Linear}, boneToModify {boneName=DYN_ThornShoulder_03}, rotation {pitch=180}}
   n19 = "Transform (Modify) Bone - Bone: DYN_ThornShoulder_02"(ComponentPose=@n18, Alpha=@n17, Translation="0,0,0", Rotation="180,0,0", Scale="1,1,1") {
    rotationMode=BMM_Additive
    rotationSpace=BCS_BoneSpace
    translationSpace=BCS_BoneSpace
    alphaBoolBlend {blendOption=Linear}
    boneToModify {boneName=DYN_ThornShoulder_02}
    rotation {pitch=180}
   }
   n20 = "Transform (Modify) Bone - Bone: DYN_ThornShoulder_01"(ComponentPose=@n19, Alpha=@n17, Translation="0,0,0", Rotation="180,0,0", Scale="1,1,1") {
    rotationMode=BMM_Additive
    rotationSpace=BCS_BoneSpace
    translationSpace=BCS_BoneSpace
    alphaBoolBlend {blendOption=Linear}
    boneToModify {boneName=DYN_ThornShoulder_01}
    rotation {pitch=180}
   }
   n21 = Get Right_Arm_ADS() -> (Right_Arm_ADS:double)
   n22 = "Transform (Modify) Bone - Bone: Dyn_Shoulder_FeathersR_Master"(ComponentPose=@n20, Alpha=@n21, Translation="-4,-2.200,-2", Rotation="2,22,-22", Scale="1,1,1") {
    rotationMode=BMM_Additive
    translationMode=BMM_Additive
    alphaBoolBlend {blendOption=Linear}
    boneToModify {boneName=Dyn_Shoulder_FeathersR_Master}
    rotation=<rot 2 22 -22>
    translation=<-4 -2.2 -2>
   }
   n23 = "Transform (Modify) Bone - Bone: Dyn_Shoulder_FeathersR_Master1"(ComponentPose=@n22, Alpha=@n21, Translation="0.000,-3,0", Rotation="0.000,0,-17", Scale="1,1,1") {
    rotationMode=BMM_Additive
    translationMode=BMM_Additive
    alphaBoolBlend {blendOption=Linear}
    boneToModify {boneName=Dyn_Shoulder_FeathersR_Master1}
    rotation {roll=-17}
    translation {y=-3}
   }
   n24 = Get Left_Arm_Up() -> (Left_Arm_Up:double)
   n25 = "Transform (Modify) Bone - Bone: Dyn_Shoulder_FeathersL_Master"(ComponentPose=@n23, Alpha=@n24, Translation="1,-2,-2", Rotation="0,0,-10", Scale="1,1,1") {
    rotationMode=BMM_Additive
    translationMode=BMM_Additive
    alphaBoolBlend {blendOption=Linear}
    boneToModify {boneName=Dyn_Shoulder_FeathersL_Master}
    rotation {roll=-10}
    translation=<1 -2 -2>
   }
   n26 = "Transform (Modify) Bone - Bone: Dyn_Shoulder_FeathersL_Master1"(ComponentPose=@n25, Alpha=@n24, Translation="0.000,0.000,-1.8", Rotation="-8,0.000,-22", Scale="1,1,1") {
    rotationMode=BMM_Additive
    translationMode=BMM_Additive
    alphaBoolBlend {blendOption=Linear}
    boneToModify {boneName=Dyn_Shoulder_FeathersL_Master1}
    rotation {pitch=-8, roll=-22}
    translation {z=-1.8}
   }
   n27 = "Transform (Modify) Bone - Bone: DYN_ShoulderLFeather"(ComponentPose=@n26, Alpha=@n24, Translation="0.2,-2,3", Rotation="0,0,0", Scale="1,1,1") {
    rotationMode=BMM_Additive
    rotationSpace=BCS_ParentBoneSpace
    translationMode=BMM_Additive
    translationSpace=BCS_ParentBoneSpace
    alphaBoolBlend {blendOption=Linear}
    boneToModify {boneName=DYN_ShoulderLFeather}
    translation=<0.2 -2 3>
   }
   n28 = Get Right_Arm_Up() -> (Right_Arm_Up:double)
   n29 = "Transform (Modify) Bone - Bone: Dyn_Shoulder_FeathersR_Master"(ComponentPose=@n27, Alpha=@n28, Translation="0.000,0.000,-0.5", Rotation="0,0,-10", Scale="1,1,1") {
    rotationMode=BMM_Additive
    translationMode=BMM_Additive
    alphaBoolBlend {blendOption=Linear}
    boneToModify {boneName=Dyn_Shoulder_FeathersR_Master}
    rotation {roll=-10}
    translation {z=-0.5}
   }
   n30 = "Transform (Modify) Bone - Bone: Dyn_Shoulder_FeathersR_Master1"(ComponentPose=@n29, Alpha=@n28, Translation="1,-3,0", Rotation="0,-15,0.000", Scale="1,1,1") {
    rotationMode=BMM_Additive
    translationMode=BMM_Additive
    alphaBoolBlend {blendOption=Linear}
    boneToModify {boneName=Dyn_Shoulder_FeathersR_Master1}
    rotation {yaw=-15}
    translation=<1 -3>
   }
   n31 = "To Float (Boolean)"(InBool=Get bIsTargeting())
   n32 = "Constraint - DYN_Jacket01_6 (PP)"!(ComponentPose=@n30, Alpha=@n31, ConstraintWeights_0=1, ConstraintWeights_1:float=0) -> (Pose) {
    alphaBoolBlend {blendOption=Linear}
    alphaScaleBiasClamp {scale=0.3}
    boneToModify {boneName=DYN_Jacket01_6}
    constraintSetup [{targetBone {boneName=spine_01}}, {targetBone {boneName=spine_05}}]
    constraintWeights [1,1]
   }
   n33 = "Constraint - DYN_Jacket06_2 (PP)"!(ComponentPose=@n32, Alpha=@n31, ConstraintWeights_0=1, ConstraintWeights_1:float=0) -> (Pose) {
    alphaBoolBlend {blendOption=Linear}
    alphaScaleBiasClamp {scale=0.3}
    boneToModify {boneName=DYN_Jacket06_2}
    constraintSetup [{targetBone {boneName=spine_01}}, {targetBone {boneName=spine_05}}]
    constraintWeights [1,1]
   }
   n34 = "Constraint - DYN_Jacket06_2 (PP)"(ComponentPose=@n33, Alpha="To Float (Boolean)"(InBool=Get bIsJumping()), ConstraintWeights_0=1, ConstraintWeights_1=0) {
    alphaBoolBlend {blendOption=Linear}
    alphaScaleBiasClamp {scale=0.2}
    boneToModify {boneName=DYN_Jacket06_2}
    constraintSetup [{targetBone {boneName=spine_01}}, {targetBone {boneName=spine_05}}]
    constraintWeights [1,1]
   }
   n35 = "To Float (Boolean)"(InBool=Get bIsCrouching())
   n36 = "Constraint - DYN_Jacket06_6 (PP)"!(ComponentPose=@n34, Alpha=@n35, ConstraintWeights_0=1, ConstraintWeights_1:float=0) -> (Pose) {
    alphaBoolBlend {blendOption=Linear}
    alphaScaleBiasClamp {scale=0.4}
    boneToModify {boneName=DYN_Jacket06_6}
    constraintSetup [{targetBone {boneName=pelvis}}, {targetBone {boneName=thigh_l}}]
    constraintWeights [1,1]
   }
   n37 = "Constraint - DYN_Jacket06_5 (PP)"!(ComponentPose=@n36, Alpha=@n35, ConstraintWeights_0=1, ConstraintWeights_1:float=0) -> (Pose) {
    alphaBoolBlend {blendOption=Linear}
    alphaScaleBiasClamp {scale=0.3}
    boneToModify {boneName=DYN_Jacket06_5}
    constraintSetup [{targetBone {boneName=spine_01}}, {targetBone {boneName=spine_05}}]
    constraintWeights [1,1]
   }
   n38 = "Transform (Modify) Bone - Bone: DYN_Jacket06_6"(ComponentPose=@n37, Alpha=@n35, Translation="0.000,-0.3,0.000", Rotation="0.000,0.000,5", Scale="1,1,1") {
    rotationMode=BMM_Additive
    translationMode=BMM_Additive
    alphaBoolBlend {blendOption=Linear}
    boneToModify {boneName=DYN_Jacket06_6}
    rotation {roll=5}
    translation {y=-0.3}
   }
   n39 = "Transform (Modify) Bone - Bone: DYN_Jacket06_5"(ComponentPose=@n38, Alpha=@n35, Translation="0.000,-0.2,0.000", Rotation="0.000,0.000,10", Scale="1,1,1") {
    rotationMode=BMM_Additive
    translationMode=BMM_Additive
    alphaBoolBlend {blendOption=Linear}
    boneToModify {boneName=DYN_Jacket06_5}
    rotation {roll=10}
    translation {y=-0.2}
   }
   n40 = Get Right_Arm_ADS()
   n41 = "Transform (Modify) Bone - Bone: DYN_FlowerShoulder_02"(ComponentPose=@n39, Alpha=@n40, Translation="1.5,0.000,0.000", Rotation="0,0,0", Scale="1,1,1") {
    rotationMode=BMM_Additive
    rotationSpace=BCS_BoneSpace
    translationMode=BMM_Additive
    translationSpace=BCS_BoneSpace
    alphaBoolBlend {blendOption=Linear}
    boneToModify {boneName=DYN_FlowerShoulder_02}
    translation {x=1.5}
   }
   n42 = "Transform (Modify) Bone - Bone: DYN_FlowerShoulder_03"(ComponentPose=@n41, Alpha=@n40, Translation="2,0.000,0.000", Rotation="0,0,0", Scale="1,1,1") {
    rotationMode=BMM_Additive
    rotationSpace=BCS_BoneSpace
    translationMode=BMM_Additive
    translationSpace=BCS_BoneSpace
    alphaBoolBlend {blendOption=Linear}
    boneToModify {boneName=DYN_FlowerShoulder_03}
    translation {x=2}
   }
   n43 = "Transform (Modify) Bone - Bone: DYN_FlowerShoulder_04"(ComponentPose=@n42, Alpha=@n40, Translation="1,0.350,1.000", Rotation="0,0,0", Scale="1,1,1") {
    rotationMode=BMM_Additive
    rotationSpace=BCS_BoneSpace
    translationMode=BMM_Additive
    translationSpace=BCS_BoneSpace
    alphaBoolBlend {blendOption=Linear}
    boneToModify {boneName=DYN_FlowerShoulder_04}
    translation=<1 0.35 1>
   }
   n44 = "Transform (Modify) Bone - Bone: DYN_FlowerShoulder_03"(ComponentPose=@n43, Alpha=1, Translation="0,0,0", Rotation="17,-14,0.000", Scale="1,1,1") {
    rotationMode=BMM_Additive
    rotationSpace=BCS_BoneSpace
    translationMode=BMM_Additive
    translationSpace=BCS_BoneSpace
    alphaBoolBlend {blendOption=Linear}
    boneToModify {boneName=DYN_FlowerShoulder_03}
    rotation {pitch=17, yaw=-14}
   }
   n45 = "Transform (Modify) Bone - Bone: DYN_FlowerShoulder_04"(ComponentPose=@n44, Alpha=1, Translation="0.78,0.35,0.000", Rotation="0,0,0", Scale="1,1,1") {
    rotationMode=BMM_Additive
    rotationSpace=BCS_BoneSpace
    translationMode=BMM_Additive
    translationSpace=BCS_BoneSpace
    alphaBoolBlend {blendOption=Linear}
    boneToModify {boneName=DYN_FlowerShoulder_04}
    translation=<0.78 0.35>
   }
   n46 = Component To Local(ComponentPose=@n45)
   n47 = OR Boolean(A=Get bIsSkyDiving(), B=Get bIsParachuteOpened())
   n48 = "Transform (Modify) Bone - Bone: dyn_glute_l"(
    ComponentPose=Local To Component(LocalPose=Blend Poses by bool(BlendPose_0=@n10, BlendPose_1=@n46, bActiveValue=@n47) {blendType=Linear})
    Alpha="To Float (Boolean)"(InBool=Get bIsCrouching())
    Translation="0.000,-4,-4"
    Rotation="0,0,0"
    Scale="1,1,1"
   ) {translationMode=BMM_Additive, alphaBoolBlend {blendOption=Linear}, boneToModify {boneName=dyn_glute_l}, translation {y=-4, z=-4}}
   n49 = Get Head_Left() -> (Head_Left:double)
   n50 = "Transform (Modify) Bone - Bone: DYN_FlowerShoulder_06"(ComponentPose=@n48, Alpha=@n49, Translation="-0.17,0.22,0", Rotation="0.000,0.000,20", Scale="1,1,1") {
    rotationMode=BMM_Additive
    translationMode=BMM_Additive
    translationSpace=BCS_ParentBoneSpace
    alphaBoolBlend {blendOption=Linear}
    boneToModify {boneName=DYN_FlowerShoulder_06}
    rotation {roll=20}
    translation=<-0.17 0.22>
   }
   n51 = "Transform (Modify) Bone - Bone: DYN_FlowerShoulder_07"(ComponentPose=@n50, Alpha=@n49, Translation="0,0,0", Rotation="0.000,14,0.000", Scale="1,1,1") {
    rotationMode=BMM_Additive
    translationMode=BMM_Additive
    translationSpace=BCS_ParentBoneSpace
    alphaBoolBlend {blendOption=Linear}
    boneToModify {boneName=DYN_FlowerShoulder_07}
    rotation {yaw=14}
   }
   n52 = "Transform (Modify) Bone - Bone: DYN_FlowerShoulder_08"(ComponentPose=@n51, Alpha=@n49, Translation="-0.9,-0.07,0", Rotation="15.2,-15.6,22", Scale="1,1,1") {
    rotationMode=BMM_Additive
    translationMode=BMM_Additive
    translationSpace=BCS_ParentBoneSpace
    alphaBoolBlend {blendOption=Linear}
    boneToModify {boneName=DYN_FlowerShoulder_08}
    rotation=<rot 15.2 -15.6 22>
    translation=<-0.9 -0.07>
   }
   n53 = "Transform (Modify) Bone - Bone: DYN_FlowerShoulder_11"(ComponentPose=@n52, Alpha=@n49, Translation="1,0,0.000", Rotation="-20,-20,40", Scale="1,1,1") {
    rotationMode=BMM_Additive
    translationMode=BMM_Additive
    alphaBoolBlend {blendOption=Linear}
    boneToModify {boneName=DYN_FlowerShoulder_11}
    rotation=<rot -20 -20 40>
    translation {x=1}
   }
   n54 = "Transform (Modify) Bone - Bone: DYN_FlowerShoulder_12"(ComponentPose=@n53, Alpha=@n49, Translation="0,0,0", Rotation="-17,-15,35", Scale="1,1,1") {
    rotationMode=BMM_Additive
    translationMode=BMM_Additive
    alphaBoolBlend {blendOption=Linear}
    boneToModify {boneName=DYN_FlowerShoulder_12}
    rotation=<rot -17 -15 35>
   }
   n55 = "Transform (Modify) Bone - Bone: DYN_FlowerShoulder_08"(ComponentPose=@n54, Alpha=1, Translation="-1.4,0.2,0", Rotation="-0.5,-3,-11", Scale="1,1,1") {
    rotationMode=BMM_Additive
    translationMode=BMM_Additive
    translationSpace=BCS_ParentBoneSpace
    alphaBoolBlend {blendOption=Linear}
    boneToModify {boneName=DYN_FlowerShoulder_08}
    rotation=<rot -0.5 -3 -11>
    translation=<-1.4 0.2>
   }
   n56 = Get Head_Up() -> (Head_Up:double)
   n57 = "Transform (Modify) Bone - Bone: DYN_FlowerShoulder_06"(ComponentPose=@n55, Alpha=@n56, Translation="-1.2,0.2,0", Rotation="-22.5,3,-1.1", Scale="1,1,1") {
    rotationMode=BMM_Additive
    translationMode=BMM_Additive
    translationSpace=BCS_ParentBoneSpace
    alphaBoolBlend {blendOption=Linear}
    boneToModify {boneName=DYN_FlowerShoulder_06}
    rotation=<rot -22.5 3 -1.1>
    translation=<-1.2 0.2>
   }
   n58 = "Transform (Modify) Bone - Bone: DYN_FlowerShoulder_07"(ComponentPose=@n57, Alpha=1, Translation="0,0,0", Rotation="0,0,0", Scale="1,1,1") {
    rotationMode=BMM_Additive
    translationMode=BMM_Additive
    translationSpace=BCS_ParentBoneSpace
    alphaBoolBlend {blendOption=Linear}
    boneToModify {boneName=DYN_FlowerShoulder_07}
   }
   n59 = "Transform (Modify) Bone - Bone: DYN_FlowerShoulder_08"(ComponentPose=@n58, Alpha=@n56, Translation="-0.5,-0.04,0.000", Rotation="11,0.5,14", Scale="1,1,1") {
    rotationMode=BMM_Additive
    translationMode=BMM_Additive
    translationSpace=BCS_ParentBoneSpace
    alphaBoolBlend {blendOption=Linear}
    boneToModify {boneName=DYN_FlowerShoulder_08}
    rotation=<rot 11 0.5 14>
    translation=<-0.5 -0.04>
   }
   n60 = "Transform (Modify) Bone - Bone: DYN_FlowerShoulder_09"(ComponentPose=@n59, Alpha=@n56, Translation="0.000,1,0", Rotation="0,0,0", Scale="1,1,1") {
    rotationMode=BMM_Additive
    translationMode=BMM_Additive
    translationSpace=BCS_ParentBoneSpace
    alphaBoolBlend {blendOption=Linear}
    boneToModify {boneName=DYN_FlowerShoulder_09}
    translation {y=1}
   }
   n61 = "Transform (Modify) Bone - Bone: DYN_Knee"(ComponentPose=@n60, Alpha="To Float (Boolean)"(InBool=Get bIsCrouching()), Translation="0,0,0", Rotation="0,0,0", Scale="1.000,1.000,0.6") {scaleMode=BMM_Additive, scaleSpace=BCS_ParentBoneSpace, alphaBoolBlend {blendOption=Linear}, boneToModify {boneName=DYN_Knee}, scale {z=0.6}}
   Output Pose(Result=Component To Local(ComponentPose=@n61)) {layerGroup=None}
  ]}
 EventGraph {nodes [
   Event BlueprintUpdateAnimation()
   > n1 = Sequence()
   n2 = BreakTransform(InTransform=GetDeltaTransformFromRefPose(self=GetOwnerSkeletalMeshComponent(), BoneName=upperarm_r, BaseName=spine_05))
   > Set right_arm_up(right_arm_up:double=MapRangeClamped(Value=BreakRotator(InRot=@n2.Rotation).Yaw, InRangeA=90, InRangeB=140)) -> (Output_Get:double)
   n3 = BreakTransform(InTransform=GetDeltaTransformFromRefPose(self=GetOwnerSkeletalMeshComponent(), BoneName=clavicle_l, BaseName=spine_05))
   > Set left_arm_up(left_arm_up:double=MapRangeClamped(Value=BreakRotator(InRot=@n3.Rotation).Pitch, InRangeA=30, InRangeB=40)) -> (Output_Get:double)
   n4 = BreakTransform(InTransform=GetDeltaTransformFromRefPose(self=GetOwnerSkeletalMeshComponent(), BoneName=upperarm_r, BaseName=spine_05))
   > Set Right_Arm_ADS(Right_Arm_ADS:double=MapRangeClamped(Value=BreakRotator(InRot=@n4.Rotation).Yaw, InRangeA=40, InRangeB=109)) -> (Output_Get:double)
   n5 = BreakTransform(InTransform=GetDeltaTransformFromRefPose(self=GetOwnerSkeletalMeshComponent(), BoneName=clavicle_l, BaseName=spine_05))
   > Set Left_Arm_ADS(Left_Arm_ADS:double=MapRangeClamped(Value=BreakRotator(InRot=@n5.Rotation).Pitch, InRangeA=-5, InRangeB=30)) -> (Output_Get:double)
   n6 = BreakTransform(InTransform=GetDeltaTransformFromRefPose(self=GetOwnerSkeletalMeshComponent(), BoneName=upperarm_r, BaseName=spine_05))
   > Set Right_Arm_ADS_Down(Right_Arm_ADS_Down:double=MapRangeClamped(Value=BreakRotator(InRot=@n6.Rotation).Pitch, InRangeA=-15, InRangeB=0)) -> (Output_Get:double)
   n7 = GetOwnerSkeletalMeshComponent()
   n8 = BreakRotator(InRot=BreakTransform(InTransform=GetDeltaTransformFromRefPose(self=@n7, BoneName=head, BaseName=spine_05)).Rotation)
   Set Head_Left(execute=@n1.then_1, Head_Left:double=MapRangeClamped(Value=@n8, InRangeA=10, InRangeB=50)) -> (Output_Get:double)
   > Set Head_Right(Head_Right:double=MapRangeClamped(Value=@n8, InRangeA=0, InRangeB=-25)) -> (Output_Get:double)
   > Set Head_Up(Head_Up:double=MapRangeClamped(Value=@n8.Yaw, InRangeA=0, InRangeB=-45)) -> (Output_Get:double)
  ]}
]
variables [
 Head_Left {type=number}
 Head_Right {type=number}
 Head_Up {type=number}
 Left_Arm_ADS {type=number}
 Left_Arm_Up {type=number}
 Right_Arm_ADS {type=number}
 Right_Arm_ADS_Down {type=number}
 Right_Arm_Up {type=number}
]
