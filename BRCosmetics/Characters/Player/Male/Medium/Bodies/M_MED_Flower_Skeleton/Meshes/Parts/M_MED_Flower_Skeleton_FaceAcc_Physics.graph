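# Physics asset for the face accessory of the male medium Flower Skeleton cosmetic, defining bodies and joints for its hood and scarf bones.
# Three two-segment box chains and a capsule tip form the hood and attach to a head capsule, while a two-segment scarf chain attaches to spine_05.
# Tight, mostly asymmetric swing and twist limits restrict hood and scarf motion, while the head-to-spine_05 joint leaves all rotation free.
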PhysicsAsset /BRCosmetics/Characters/Player/Male/Medium/Bodies/M_MED_Flower_Skeleton/Meshes/Parts/M_MED_Flower_Skeleton_FaceAcc_Physics
bodies {
 DYN_Hood_01_1 [{shapeType=Box, extentX=4.9375, extentY=4.4375, extentZ=1, center=<2.8008 0 0>, rotation=<rot 0 0 -8.4388>}]
 DYN_Hood_01_2 [{shapeType=Box, extentX=3.375, extentY=4.375, extentZ=1, center=<1.868 0 0>, rotation=<rot 0 0 -5.6282>}]
 DYN_Hood_02_1 [{shapeType=Box, extentX=5, extentY=1, extentZ=5.0625, center=<2.6726 0 0>}]
 DYN_Hood_02_2 [{shapeType=Box, extentX=4, extentY=1, extentZ=5, center=<2.1531 0 0>}]
 DYN_Hood_03_1 [{shapeType=Box, extentX=5.0625, extentY=5.5625, extentZ=1, center=<2.7245 0 0>, rotation=<rot 0 0 11.2533>}]
 DYN_Hood_03_2 [{shapeType=Box, extentX=3.4375, extentY=5.625, extentZ=1, center=<1.8363 0 0>, rotation=<rot 0 0 11.2531>}]
 DYN_Hood_Tip [{shapeType=Capsule, length=0.875, radius=1.5625, center=<1.9806 0 0>, rotation=<rot 90 0 0>}]
 DYN_Scarf1 [{shapeType=Box, extentX=8.9375, extentY=1, extentZ=10.8125, center=<4.7663 0 0>}]
 DYN_Scarf2 [{shapeType=Box, extentX=7.3125, extentY=1, extentZ=10.75, center=<3.7589 0.0002 0>}]
 head [{shapeType=Capsule, length=11.5796, radius=6.5804, center=<1.6996 4.7774 -0.0851>, rotation=<rot -0.7689 102.1924 -89.1512>}]
 spine_05 [{shapeType=Box, extentX=1, extentY=1, extentZ=1}]
}
constraints [
 {bone1Name=DYN_Hood_01_2, bone2Name=DYN_Hood_01_1, swing1LimitDegrees=2, swing2LimitDegrees=30, twistLimitDegrees=7}
 {bone1Name=DYN_Hood_03_2, bone2Name=DYN_Hood_03_1, swing1LimitDegrees=2, swing2LimitDegrees=30, twistLimitDegrees=7}
 {bone1Name=DYN_Hood_02_2, bone2Name=DYN_Hood_02_1, swing1LimitDegrees=15, swing2LimitDegrees=10, twistLimitDegrees=7}
 {bone1Name=DYN_Scarf2, bone2Name=DYN_Scarf1, swing1LimitDegrees=20, swing2LimitDegrees=20, twistLimitDegrees=5}
 {bone1Name=DYN_Hood_01_1, bone2Name=head, swing1LimitDegrees=2, swing2LimitDegrees=30, twistLimitDegrees=7}
 {bone1Name=DYN_Hood_03_1, bone2Name=head, swing1LimitDegrees=2, swing2LimitDegrees=30, twistLimitDegrees=7}
 {bone1Name=DYN_Hood_Tip, bone2Name=head, swing1LimitDegrees=15, swing2LimitDegrees=15, twistLimitDegrees=5}
 {bone1Name=DYN_Hood_02_1, bone2Name=head, swing1LimitDegrees=10, swing2LimitDegrees=2, twistLimitDegrees=7}
 {
  bone1Name=head
  bone2Name=spine_05
  swing1LimitDegrees=45
  swing1Motion=Free
  swing2LimitDegrees=45
  swing2Motion=Free
  twistLimitDegrees=45
  twistMotion=Free
 }
 {bone1Name=DYN_Scarf1, bone2Name=spine_05, swing1LimitDegrees=20, swing2LimitDegrees=20, twistLimitDegrees=5}
]
