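# Animation blueprint for the Flower Skeleton male outfit's face accessory, adding physics and head-driven correction to its scarf.
# It copies the attached parent mesh's pose and curves, then runs rigid body simulation in root bone space with overridden gravity.
# Each update maps head rotation relative to spine_05 into clamped up, left and right weights.
# These weights drive three chained additive rotations on DYN_ScarfBase, applied after the simulation.
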AnimBlueprint /BRCosmetics/Characters/Player/Male/Medium/Bodies/M_MED_Flower_Skeleton/Meshes/Parts/M_MED_Flower_Skeleton_FaceAcc_AnimBP
target_skeleton=&/M_MED_Flower_Skeleton_FaceAcc_Skeleton
graphs [
 AnimGraph {nodes [
   n1 = RigidBody(ComponentPose=Local To Component(LocalPose=Copy Pose From Mesh() {bCopyCurves=true, bUseAttachedParent=true})) {
    bOverrideWorldGravity=true
    overridePhysicsAsset=&/M_MED_Flower_Skeleton_FaceAcc_Physics
    simulationSpace=BaseBoneSpace
    alphaBoolBlend {blendOption=Linear}
    baseBoneRef {boneName=root}
    componentLinearAccScale=<0.3 0.3 0.3>
    componentLinearVelScale=<0.6 0.6 0.6>
    overrideWorldGravity {z=-980}
   }
   n2 = "Transform (Modify) Bone - Bone: DYN_ScarfBase"(ComponentPose=@n1, Alpha=Get Head_Up() -> (Head_Up:double), Translation="0,0,0", Rotation="0.000,0.000,19", Scale="1,1,1") {
    rotationMode=BMM_Additive
    translationMode=BMM_Additive
    alphaBoolBlend {blendOption=Linear}
    boneToModify {boneName=DYN_ScarfBase}
    rotation {roll=19}
   }
   n3 = "Transform (Modify) Bone - Bone: DYN_ScarfBase"(ComponentPose=@n2, Alpha=Get Head_Left() -> (Head_Left:double), Translation="0,0,0", Rotation="0.000,20,0", Scale="1,1,1") {
    rotationMode=BMM_Additive
    translationMode=BMM_Additive
    alphaBoolBlend {blendOption=Linear}
    boneToModify {boneName=DYN_ScarfBase}
    rotation {yaw=20}
   }
   n4 = "Transform (Modify) Bone - Bone: DYN_ScarfBase"(ComponentPose=@n3, Alpha=Get Head_Right() -> (Head_Right:double), Translation="0,0,0", Rotation="0.000,-20,0", Scale="1,1,1") {
    rotationMode=BMM_Additive
    translationMode=BMM_Additive
    alphaBoolBlend {blendOption=Linear}
    boneToModify {boneName=DYN_ScarfBase}
    rotation {yaw=-20}
   }
   Output Pose(Result=Component To Local(ComponentPose=@n4)) {layerGroup=None}
  ]}
 EventGraph {nodes [
   Event BlueprintUpdateAnimation()
   > Sequence()
   n1 = GetOwnerSkeletalMeshComponent()
   n2 = BreakRotator(InRot=BreakTransform(InTransform=GetDeltaTransformFromRefPose(self=@n1, BoneName=head, BaseName=spine_05)).Rotation)
   > Set Head_Left(Head_Left:double=MapRangeClamped(Value=@n2, InRangeA=10, InRangeB=50)) -> (Output_Get:double)
   > Set Head_Right(Head_Right:double=MapRangeClamped(Value=@n2, InRangeA=0, InRangeB=-25)) -> (Output_Get:double)
   > Set Head_Up(Head_Up:double=MapRangeClamped(Value=@n2.Yaw, InRangeA=0, InRangeB=-45)) -> (Output_Get:double)
   PrintString!(execute, InString="To String (Float)"(InDouble=0), bPrintToScreen, bPrintToLog, TextColor, Duration, Key) -> (then)
  ]}
]
variables [
 Head_Left {type=number}
 Head_Right {type=number}
 Head_Up {type=number}
]
