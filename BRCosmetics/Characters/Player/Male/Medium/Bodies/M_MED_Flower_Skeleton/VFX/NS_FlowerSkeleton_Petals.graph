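# Niagara effect for the Flower Skeleton player body cosmetic that sheds occasional sprite petals from the character's skeletal mesh.
# The sole enabled emitter, Petals, samples User.SkeletalMesh, inherits capped velocity, then drifts petals with gravity, drag and panning curl noise.
# Sine-driven width fakes flutter, while a threshold from scalability distance fraction shrinks petals and zeroes their lifetime to cull them.
# A mesh-petal variant and two placeholder-named emitters are disabled.
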NiagaraSystem /BRCosmetics/Characters/Player/Male/Medium/Bodies/M_MED_Flower_Skeleton/VFX/NS_FlowerSkeleton_Petals
dependencies {usedDataInterfaces [&NiagaraDataInterfaceCurve], usedDynamicInputs [], usedModules [&SystemState]}
emitters {
 NE_Empty {
  bEnabled=false
  emitterUpdateScript [
   &EmitterState(Life Cycle Mode=::NewEnumerator1, Loop Duration=1, UseLoopDelay=false)
   &SpawnRate(SpawnRate=20, Use Spawn Probability=false)
  ]
  particleSpawnScript [
   /Niagara/Modules/Spawn/Initialization/InitializeParticle(
    Lifetime=5
    Module.Write Color=true
    Module.Write Lifetime=true
    Module.Write Mass=true
    Module.Write Material Random=true
    Module.Write Position=true
    Module.Write Scale=false
    Module.Write SpriteRotation=false
    Module.Write SpriteSize=true
    Position=&Add_Vector(A=/Niagara/DynamicInputs/Helpers/SimulationPosition(), B=<0 0 0>)
    Sprite Size=<5 5>
   )
   &BoxLocation(Box Offset=<0 0 100>, MaskSpawnGroup=false, Module.Override Seed=false, Rotate Box=false, Surface Only=false)
  ]
  particleUpdateScript [&ParticleState()]
  rendererClasses [&NiagaraSpriteRendererProperties]
  renderers [{rendererClass=&NiagaraSpriteRendererProperties}]
 }
 NE_Empty001 {
  bEnabled=false
  emitterUpdateScript [
   &EmitterState(Life Cycle Mode=::NewEnumerator1, Loop Duration=1, UseLoopDelay=false)
   &SpawnBurst_Instantaneous(Spawn Count=1, Use Spawn Probability=false)
  ]
  particleSpawnScript [&InitializeParticle(
    Lifetime=5
    Module.Write Color=true
    Module.Write Lifetime=true
    Module.Write Mass=true
    Module.Write Material Random=true
    Module.Write Position=true
    Module.Write Scale=false
    Module.Write SpriteRotation=false
    Module.Write SpriteSize=true
    Position=&Add_Vector(A=&SimulationPosition(), B=<-69.2746 -100 0>)
    Sprite Size=&Vector2DFromFloat(Value=100)
   )]
  particleUpdateScript [
   &ParticleState(Kill Particles When Lifetime Has Elapsed=false)
   &DynamicMaterialParameters(
    Index 0 Param 1=&MakeCustomFloatFromBool(
     Bool=&SetBoolByFloatComparison(
      A=0.3
      B=&Power(
       a=&ClampFloat(Float=&Subtract_Float(
         A=&Multiply_Float(A=9.1, B=&OneMinusFloat(Float=$Emitter.Scalability.DistanceFraction))
         B=8
        ))
       b=4
      )
      Comparison Type=::NewEnumerator6
     )
     True Float=&Power(
      a=&ClampFloat(Float=&Subtract_Float(
        A=&Multiply_Float(A=9.1, B=&OneMinusFloat(Float=$Emitter.Scalability.DistanceFraction))
        B=8
       ))
      b=4
     )
     False Float=0
    )
    Module.Param0WriteEnabled=true
    Module.Param1WriteEnabled=true
    Module.Param2WriteEnabled=true
    Module.Param3WriteEnabled=true
   )
  ]
  rendererClasses [&NiagaraSpriteRendererProperties]
  renderers [{rendererClass=&NiagaraSpriteRendererProperties}]
 }
 Petals {
  emitterUpdateScript [
   &EmitterState()
   &SpawnRate(Spawn Probability=0.05, SpawnRate=40, Use Spawn Probability=true)
  ]
  particleSpawnScript [
   &InitializeParticle(
    Color=<rgba 0 0 0 1>
    Lifetime=3
    Module.Write Color=true
    Module.Write Lifetime=true
    Module.Write Mass=true
    Module.Write Material Random=true
    Module.Write Position=false
    Module.Write Scale=false
    Module.Write SpriteRotation=true
    Module.Write SpriteSize=true
    Position=/Niagara/Functions/Localspace/SimulationPosition()
    Sprite Rotation=&UniformRangedFloat(Minimum=0, Maximum=360)
    Sprite Size=&Multiply_Vector2(A=<2.5 2.5>, B=&UniformRangedVector2D(Minimum=<1 0.5>, Maximum=<2.5 1.6>))
   )
   &SphereLocation(
    Hemisphere X=true
    MaskSpawnGroup=false
    Non Uniform Scale=<0 1.2 1.8>
    Override Seed=false
    Sphere Radius=10
    Surface Only=false
    UseNonUniformScale=true
    UseOffset=false
    UseSphereOrientation=false
   )
   &AddVelocity(Velocity=&UniformRangedVector(Minimum=<3 -10 -5>, Maximum=<7 10 1>))
   &InheritVelocity(Module.Limit Speed=true, Velocity Limit=75)
   /SetVariables(Particles.stableGravity:NiagaraFloat=&UniformRangedFloat(Minimum=-20, Maximum=-40)) {isSetParameters=true, module=SetVariables_#1}
   /SetVariables(Particles.stableRotationRate:NiagaraFloat=&UniformRangedFloat(Minimum=50, Maximum=200)) {isSetParameters=true, module=SetVariables_#2}
   &AddVelocityFromPoint(Module.Use Falloff Distance=false, Origin Offset Coordinate Space=::Simulation, Velocity Strength=20) {enabled=false}
   &SkeletalMeshLocation(
    Bone Sampling Mode=::NewEnumerator0
    Inherit Velocity=false
    MaskSpawnGroup=false
    Module.Override Randomness=false
    Skeletal Mesh=$User.SkeletalMesh
   )
  ]
  particleUpdateScript [
   /SetVariables(Particles.DistanceThreshold:NiagaraFloat=&MakeCustomFloatFromBool(
     Bool=&SetBoolByFloatComparison(
      A=0.3
      B=&Power(
       a=&ClampFloat(Float=&Subtract_Float(
         A=&Multiply_Float(A=9.1, B=&OneMinusFloat(Float=$Emitter.Scalability.DistanceFraction))
         B=8
        ))
       b=4
      )
      Comparison Type=::NewEnumerator6
     )
     True Float=&Power(
      a=&ClampFloat(Float=&Subtract_Float(
        A=&Multiply_Float(A=9.1, B=&OneMinusFloat(Float=$Emitter.Scalability.DistanceFraction))
        B=8
       ))
      b=4
     )
     False Float=0
    )) {isSetParameters=true, module=SetVariables_#3}
   &ParticleState(Lifetime:NiagaraFloat=&MakeCustomFloatFromBool(
     Bool=&SetBoolByFloatComparison(A=$Particles.DistanceThreshold, B=0, Comparison Type=::NewEnumerator7)
     True Float=0
     False Float=$Particles.Lifetime
    ))
   &GravityForce(Gravity=&MakeVector(X=0, Y=0, Z=$Particles.Initial.stableGravity))
   &Drag(Drag=&UniformRangedFloat(Minimum=0.8, Maximum=1.25), Use Linear Drag=true, Use Rotational Drag=false)
   /Niagara/Modules/Update/Forces/CurlNoiseForce(Noise Frequency=0.1, Noise Strength=150, Pan Noise=true, Pan Noise Field=<0 0 -50>)
   &SolveForcesAndVelocity(Clamp Velocity=false, Limit Acceleration=false)
   &SpriteRotationRate(Rotation Rate=$Particles.Initial.stableRotationRate)
   &ScaleSpriteSize(Scale Factor=&Multiply_Vector2(
     A=&Multiply_Vector2(
      A=&MakeVector2D(
       X=&Multiply_Float(
        A=1
        B=&Sine(
         Normalized Angle=$Emitter.Age
         Period=&Multiply_Float(A=3, B=&Add_Float(A=$Particles.Initial.MaterialRandom, B=0.6))
         Scale=1
         Bias=0
        )
       )
       Y=1
      )
      B=&Vector2DFromFloat(Value=&FloatFromCurve(FloatCurve=DataInterface() {propertyValues {Curve {keys [<at 0 1.025>,<at 0.9 1.025>,<at 1 0>]}}}))
     )
     B=&Vector2DFromFloat(Value=$Particles.DistanceThreshold)
    ))
  ]
  rendererClasses [&NiagaraSpriteRendererProperties]
  renderers [{rendererClass=&NiagaraSpriteRendererProperties}]
 }
 Petals001 {
  bEnabled=false
  emitterUpdateScript [
   &EmitterState()
   &SpawnRate(Spawn Probability=0.222, SpawnRate=6, Use Spawn Probability=true)
  ]
  particleSpawnScript [
   &InitializeParticle(
    Color=<rgba 0 0 0 1>
    Lifetime=5
    Mesh Scale=&VectorFromFloat(Value=&UniformRangedFloat(Minimum=0.15, Maximum=0.16))
    Module.Write Color=true
    Module.Write Lifetime=true
    Module.Write Mass=true
    Module.Write Material Random=true
    Module.Write Position=false
    Module.Write Scale=true
    Module.Write SpriteRotation=true
    Module.Write SpriteSize=true
    Position=&SimulationPosition()
    Sprite Rotation=&UniformRangedFloat(Minimum=0, Maximum=360)
    Sprite Size=&Multiply_Vector2(A=<2.5 2.5>, B=&UniformRangedVector2D(Minimum=<1 0.5>, Maximum=<2 2>))
   )
   &SphereLocation(
    Hemisphere X=true
    MaskSpawnGroup=false
    Non Uniform Scale=<0 1.2 1.8>
    Override Seed=false
    Sphere Radius=10
    Surface Only=false
    UseNonUniformScale=false
    UseOffset=false
    UseSphereOrientation=false
   )
   &AddVelocity(Velocity=&UniformRangedVector(Minimum=<0 -15 0>, Maximum=<10 15 5>))
   &InheritVelocity(Module.Limit Speed=true, Velocity Limit=75)
   /SetVariables(Particles.stableGravity:NiagaraFloat=&UniformRangedFloat(Minimum=-20, Maximum=-40)) {isSetParameters=true, module=SetVariables_#1}
   /SetVariables(Particles.stableRotationRate:NiagaraFloat=&UniformRangedFloat(Minimum=50, Maximum=200)) {isSetParameters=true, module=SetVariables_#2}
   &AddVelocityFromPoint(Module.Use Falloff Distance=false, Origin Offset Coordinate Space=::Simulation, Velocity Strength=20)
   /Niagara/Modules/Spawn/Orientation/V2/InitialMeshOrientation(
    Rotation=&UniformRangedVector(Minimum=<0 0 0>, Maximum=<1 1 1>)
    Use Orientation Vector=true
    Use Rotation Vector=true
   )
  ]
  particleUpdateScript [
   /SetVariables(Particles.DistanceThreshold:NiagaraFloat=&MakeCustomFloatFromBool(
     Bool=&SetBoolByFloatComparison(
      A=0.3
      B=&Power(
       a=&ClampFloat(Float=&Subtract_Float(
         A=&Multiply_Float(A=9.1, B=&OneMinusFloat(Float=$Emitter.Scalability.DistanceFraction))
         B=8
        ))
       b=4
      )
      Comparison Type=::NewEnumerator6
     )
     True Float=&Power(
      a=&ClampFloat(Float=&Subtract_Float(
        A=&Multiply_Float(A=9.1, B=&OneMinusFloat(Float=$Emitter.Scalability.DistanceFraction))
        B=8
       ))
      b=4
     )
     False Float=0
    )) {isSetParameters=true, module=SetVariables_#3}
   &ParticleState(Lifetime:NiagaraFloat=&MakeCustomFloatFromBool(
     Bool=&SetBoolByFloatComparison(A=$Particles.DistanceThreshold, B=0, Comparison Type=::NewEnumerator7)
     True Float=0
     False Float=$Particles.Lifetime
    ))
   &GravityForce(Gravity=&MakeVector(X=0, Y=0, Z=$Particles.Initial.stableGravity))
   &Drag(Drag=&UniformRangedFloat(Minimum=0.8, Maximum=1.25), Use Linear Drag=true, Use Rotational Drag=false)
   &CurlNoiseForce(Noise Frequency=0.065, Noise Strength=25, Pan Noise=false)
   &SolveForcesAndVelocity(Clamp Velocity=false, Limit Acceleration=false)
   &SpriteRotationRate(Rotation Rate=$Particles.Initial.stableRotationRate)
   &ScaleSpriteSize(Scale Factor=&Multiply_Vector2(
     A=&Multiply_Vector2(
      A=&MakeVector2D(
       X=&Multiply_Float(
        A=1
        B=&Sine(
         Normalized Angle=$Emitter.Age
         Period=&Multiply_Float(A=3, B=&Add_Float(A=$Particles.Initial.MaterialRandom, B=0.6))
         Scale=1
         Bias=0
        )
       )
       Y=1
      )
      B=&Vector2DFromFloat(Value=&FloatFromCurve(FloatCurve=DataInterface() {propertyValues {Curve {keys [<at 0 1>,<at 0.9 1>,<at 1 0>]}}}))
     )
     B=&Vector2DFromFloat(Value=$Particles.DistanceThreshold)
    )) {enabled=false}
   &MeshRotationForce(Rotation=<2 0 0>)
   &SolveRotationalForcesAndVelocity()
  ]
  rendererClasses [&NiagaraMeshRendererProperties]
  renderers [{rendererClass=&NiagaraMeshRendererProperties, rendererIndex=0}]
 }
}
summary {
 systemName=NS_FlowerSkeleton_Petals
 emitters [
  {emitterName=Petals, rendererClasses [&NiagaraSpriteRendererProperties]}
  {bEnabled=false, emitterName=NE_Empty, rendererClasses [&NiagaraSpriteRendererProperties]}
  {bEnabled=false, emitterName=NE_Empty001, rendererClasses [&NiagaraSpriteRendererProperties]}
  {bEnabled=false, emitterName=Petals001, rendererClasses [&NiagaraMeshRendererProperties]}
 ]
 userVariables [
  User.bIsFrontEndPreview {type=NiagaraBool}
  User.PlayerSpeed {type=NiagaraFloat, defaultValue=0, description=""}
  User.SkeletalMesh {
   type=NiagaraDataInterfaceSkeletalMesh
   defaultValue {
    struct=NiagaraExt_VariableValue_DataInterface
    value {dataInterface=&self:NiagaraDataInterfaceSkeletalMesh_1}
   }
  }
 ]
}
